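; Building prop Blueprint that shows a golf course hole number on its static mesh, although the asset name refers to a broken branch.
; Its construction script switches on the HoleNumber variable and assigns the matching golf sign number material to material slot 1.
; Only holes 1 through 9 are handled, and any other value leaves the material untouched.
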(fn ConstructionScript
  (bind _staticmeshcomponent (Variables|Debug|GetStaticMeshComponent))
  (switch int (Variables|Default|GetHoleNumber)
    (:1
      (Rendering|Material|SetMaterial _staticmeshcomponent 1 "/BlastBerryMapContent/Game/Environments/Props/Golf_Course/Materials/M_GolfCourse_Sign_Numbers_1.M_GolfCourse_Sign_Numbers_1"))
    (:2
      (Rendering|Material|SetMaterial _staticmeshcomponent 1 "/BlastBerryMapContent/Game/Environments/Props/Golf_Course/Materials/M_GolfCourse_Sign_Numbers_2.M_GolfCourse_Sign_Numbers_2"))
    (:3
      (Rendering|Material|SetMaterial _staticmeshcomponent 1 "/BlastBerryMapContent/Game/Environments/Props/Golf_Course/Materials/M_GolfCourse_Sign_Numbers_3.M_GolfCourse_Sign_Numbers_3"))
    (:4
      (Rendering|Material|SetMaterial _staticmeshcomponent 1 "/BlastBerryMapContent/Game/Environments/Props/Golf_Course/Materials/M_GolfCourse_Sign_Numbers_4.M_GolfCourse_Sign_Numbers_4"))
    (:5
      (Rendering|Material|SetMaterial _staticmeshcomponent 1 "/BlastBerryMapContent/Game/Environments/Props/Golf_Course/Materials/M_GolfCourse_Sign_Numbers_5.M_GolfCourse_Sign_Numbers_5"))
    (:6
      (Rendering|Material|SetMaterial _staticmeshcomponent 1 "/BlastBerryMapContent/Game/Environments/Props/Golf_Course/Materials/M_GolfCourse_Sign_Numbers_6.M_GolfCourse_Sign_Numbers_6"))
    (:7
      (Rendering|Material|SetMaterial _staticmeshcomponent 1 "/BlastBerryMapContent/Game/Environments/Props/Golf_Course/Materials/M_GolfCourse_Sign_Numbers_7.M_GolfCourse_Sign_Numbers_7"))
    (:8
      (Rendering|Material|SetMaterial _staticmeshcomponent 1 "/BlastBerryMapContent/Game/Environments/Props/Golf_Course/Materials/M_GolfCourse_Sign_Numbers_8.M_GolfCourse_Sign_Numbers_8"))
    (:9
      (Rendering|Material|SetMaterial _staticmeshcomponent 1 "/BlastBerryMapContent/Game/Environments/Props/Golf_Course/Materials/M_GolfCourse_Sign_Numbers_9.M_GolfCourse_Sign_Numbers_9"))))
