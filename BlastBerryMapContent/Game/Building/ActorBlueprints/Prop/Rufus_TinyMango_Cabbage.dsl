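; Rufus_TinyMango_Cabbage is a prop actor Blueprint, by name a cabbage variant in the Rufus TinyMango prop set.
; Its construction script sets the "Hide Element 01" scalar parameter to 1 on the materials of its StaticMesh1 component.
; The parameter name suggests this hides one element of the mesh through the material rather than swapping meshes.
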(fn ConstructionScript
  (Rendering|Material|SetScalarParameterValueonMaterials (Variables|Default|GetStaticMesh1) "Hide Element 01" 1.0))
